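; Campfire trap actor for Athena's environmental campfire, which switches itself to lit when it first initializes.
; On building actor initialization it calls the parent, then, if it has never been lit, sets the native replicated lit state and marks it as lit.
; BeginPlay, overlap, tick and the construction script add no logic of their own.
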(event EventBeginPlay
  (|CallParentFunction) ; This node is disabled and will not be called.\nDrag off pins to build functionality.
)

(event Collision|EventActorBeginOverlap
  (params (OtherActor :type Actor%20Object%20Reference)))

(event EventTick
  (params (DeltaSeconds :type Float%20%28single-precision%29)))

(event Building|EventOnBuildingActorInitialized
  (params (InitializationReason :type EFortBuildingInitializationReason%20Enum :default "StaticallyPlaced") (BuildingPersistentState :type EFortBuildingPersistentState%20Enum :default "Default"))
  (|CallParentFunction InitializationReason BuildingPersistentState)
  (if (not (Variables|Default|GetbHasBeenLitNative))
    (Variables|StateBools|SetLitRepBoolNative 1)
    (Variables|Default|SetbHasBeenLitNative true)))

(fn ConstructionScript
  (|CallParentFunction))
