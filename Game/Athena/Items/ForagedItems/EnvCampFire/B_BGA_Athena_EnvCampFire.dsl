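; Environmental campfire actor for Battle Royale that players light, relight or stoke with wood, driving a healing ability through gameplay events.
; On playlist ready it applies data-table hotfixes that can replace, remove, pre-light or storm-lock it.
; Curie fire attach and detach sync the lit state, while server-side sphere overlaps manage per-actor fire interactions.
; Most state logic delegates to native functions, with extra Season 7 quest code switching a prisoner skin variant.
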(event Custom|PayStokeCost
  (params (InteractingPawn :type Fort%20Pawn%20Object%20Reference))
  ;[1 1 1] Pay Stoke Cost
  (Default|RemoveItemNative InteractingPawn (Variables|Default|GetItemWoodNative) (Math|Float|Ceil (ScalableFloat|GetValueatLevel (Variables|DataRows|GetRow_Stoke_CostNative)))))

(event Custom|PayLightCost
  (params (InteractingPawn :type Fort%20Pawn%20Object%20Reference))
  ;[1 1 1] Pay Cost
  (Default|RemoveItemNative InteractingPawn (Variables|Default|GetItemWoodNative) (CallFunction|GetCostReActivate)))

(event Custom|CheckReplaceCampfire
  (bind (_returnvalue_location _returnvalue_rotation) (Math|Transform|BreakTransform (Transformation|GetActorTransform)))
  ;[1 1 1] Check if we should replace the campfire with the pot version. If so do it and then destroy this campfire.
  (if (and (CallFunction|RowtoBoolNative (Variables|DataRows|GetRow_ReplaceCampfire)) (Networking|IsServer))
    (Utilities|FlowControl|Delay (Math|Random|RandomFloat)) ; Random float to help with spawning on different frames
    (Building|SpawnBuildingActor (Variables|Default|GetActorToReplaceWith) (Math|Transform|MakeTransform _returnvalue_location _returnvalue_rotation) 0 0 false false)
    (Actor|SetLifeSpan 0.2)))

(event EventBeginPlay
  ;[1 1 1] Begin Play
  ;;[1 1 1] BeginPlay - When Playlist Ready Async, Set sheet values
  (Curie|OnCurieActivated
    (:then
      (Development|SetHiddeninGame (Variables|Default|GetWood) (== (Variables|StateBools|GetLitRepBoolNative) 0) false)
      (Class|FortAsyncActionCurrentPlaylistReady|PlaylistDataReadyAsync
        (:OnReady
          (CallFunction|CheckforRetriggerPlaylist)
          (CallFunction|Start)
          (CallFunction|CheckDestroyDisabled)
          (CallFunction|CheckReplaceCampfire)
          (CallFunction|DisableifBelowFlood)
          ; FORT-1100305 - Adding HF for campfires to start active
          (if (ScalableFloat|GetValueasBool (Variables|DataRows|GetRow_CampfireStartsActive))
            (Variables|StateBools|SetLitRepBoolNative 1)
            (else
              _))
          ;[1 1 1] FX - Wood burning material instance
          (bind _returnvalue (Rendering|Material|CreateDynamicMaterialInstance (Variables|Default|GetWood) 0 (Rendering|Material|GetMaterial (Variables|Default|GetWood) 0) "None" :!class /Script/Engine.PrimitiveComponent))
          (bind _output_get (Variables|FX|SetWoodBurningMatInstanceNative _returnvalue))
          (Rendering|Material|SetScalarParameterValue _output_get "Burn Amount" (Variables|FX|GetFX_WoodBurnAmountNative) :!class /Script/Engine.MaterialInstanceDynamic)
          (Rendering|Material|SetScalarParameterValue _output_get "Emissive" (Variables|FX|GetFX_WoodBurnEmissiveNative) :!class /Script/Engine.MaterialInstanceDynamic))))
    (:OnCurieActive
      (Variables|Default|SetCurieActivated true))))

(event Custom|CheckDestroyDisabled
  ;[1 1 1] Begin Play
  ;;[1 1 1] CheckDestroyDisabled - Check Destroy Hotfix
  (if (and (Variables|StateBools|GetDestroyIfDisabled) (not (Variables|StateBools|GetbHotfixEnabledNative)))
    (Actor|DestroyActor)))

(event Custom|Start
  ;[1 1 1] Begin Play
  ;;[1 1 1] Start - Check Sheet Values
  (CallFunction|SetSheetValues))

(event Custom|HandleBoundDestroyed
  ;[1 1 1] Check For Floor Binding
  ;;[1 1 1] Kill Actor on bound destroy and play Cues
  (Ability|ExecuteGameplayCueReplicated (Variables|GameplayCue|GetDeathCue))
  (Variables|StateBools|SetDestroyed true)
  (Collision|SetActorEnableCollision false)
  (Actor|DestroyActor))

(event Interaction|EventBlueprintOnInteract
  (params (InteractingPawn :type Fort%20Pawn%20Object%20Reference) (InteractionBeingAttempted :type EInteractionBeingAttempted%20Enum :default "FirstInteraction"))
  ;[1 1 1] Begin Play
  ;;[1 1 1] Interact - On Interact save the interacting pawn and light the fire
  (bind _output_get (Variables|Interact|SetInteractingPawnNative InteractingPawn))
  (if (Variables|StateBools|GetbHotfixEnabledNative)
    (if (== (Variables|StateBools|GetLitRepBoolNative) 1)
      ;;;[1 1 1] Stoke Fire
      (Ability|SendGameplayEventtoActor self "(TagName=\"Event.Athena.EnvCampFire.Light\")" (Utilities|Struct|MakeGameplayEventData :ContextHandle "()" :InstigatorTags "(GameplayTags=)" :TargetTags "(GameplayTags=)" :EventMagnitude 1.0 :TargetData "()"))
      (CallFunction|PayStokeCost _output_get)
      (else
        (Variables|Default|SetbHasBeenLitNative true)
        (if (Variables|StateBools|GetbBeenUsedNative)
          (CallFunction|PayLightCost _output_get))
        (Variables|StateBools|SetLitRepBoolNative 1)
        ;;;[1 1 1] We dont need this for curie stuff to work, but having fire attached to the actor passes backing data from the fire rather than needing to recreate it continously.
        (Curie|ExecuteFireApplication (Fortnite|GetFortPlayerControllerfromActor (Variables|Interact|GetInteractingPawnNative)) self self 1e+02)
        (bind _asathena_player_controller (Utilities|Casting|CastToAthena_PlayerController (Pawn|GetController (Variables|Interact|GetInteractingPawnNative))))
        (Analytics|FireAnalyticEventEnvironmentalItemUsed _asathena_player_controller self)
        (bind _userisspecial (CallFunction|CheckifUserIsSpecialS7 (Variables|Interact|GetInteractingPawnNative)))
        (if _userisspecial
          (EventDispatchers|BindEventtoOnQuestObjectiveStateChanged _asathena_player_controller (EventDispatchers|CreateEvent :!event Custom|Server_QuestObjectiveUpdated)))))))

(event OnLinkedActorDestroyed(FortLinkToActor)
  (params (DamageCauser :type Actor%20Object%20Reference))
  ;[1 1 1] Check For Floor Binding
  ;;[1 1 1] Linked Actor Destroyed - Destroy linked actor.
  (CallFunction|HandleBoundDestroyed))

(event Custom|Light
  ;[1 1 1] Light and Go Out
  ;;[1 1 1] Light - Set lit to True. Send event to trgger ability that heals. Apply Fire GC
  (CallFunction|LightNative))

(event Custom|GoOut
  ;[1 1 1] Light and Go Out
  ;;[1 1 1] GoOut - Set Lit to false and toggle the fire pit (Called By GA by setting the IsLit bool). Reset Wood Material FX when fire burns out. Add and remove GCs, Wait a duration then remove the doused too.
  (CallFunction|GoOutNative))

(event Curie|EventOnCurieElementAttached
  (params (CurieContainerHandle :type Curie%20Container%20Handle%20Structure) (ElementTag :type Gameplay%20Tag%20Structure))
  ;[0.769 0.09549 1] Curie
  (if (Variables|Default|GetCurieActivated)
    ;;[1 1 1] CurieToggleActivate - Fires when campfire is started from an external source
    (if (GameplayTags|Equal(GameplayTag) ElementTag "(TagName=\"Curie.Element.Fire\")")
      (if (not (Variables|Default|GetbHasBeenLitNative))
        (Variables|Default|SetbHasBeenLitNative true)
        (Variables|StateBools|SetLitRepBoolNative 1)))))

(event Curie|EventOnCurieElementDetached
  (params (CurieContainerHandle :type Curie%20Container%20Handle%20Structure) (ElementTag :type Gameplay%20Tag%20Structure))
  (bind _self self)
  ;[0.769 0.09549 1] Curie
  ;;[0.126 1 0] Called if the fire is made inactive.  Ends heal ability and removes Curie interaction
  (if (Math|Boolean|ANDBoolean (GameplayTags|Equal(GameplayTag) ElementTag "(TagName=\"Curie.Element.Fire\")") (== (Variables|StateBools|GetLitRepBoolNative) 1) (Variables|Default|GetCurieActivated))
    (Ability|SendGameplayEventtoActor _self "(TagName=\"Event.Athena.EnvCampFire.Douse\")" (Utilities|Struct|MakeGameplayEventData :ContextHandle "()" :InstigatorTags "(GameplayTags=)" :TargetTags "(GameplayTags=)" :EventMagnitude 0.0 :TargetData "()"))
    (CallFunction|EndFireInteract)
    (else
      (Ability|SendGameplayEventtoActor _self "(TagName=\"Event.Athena.EnvCampFire.Expired\")"))))

(event Custom|CheckInitialOverlappingActors
  ;[1 1 1] Check Overlapping Actors on Ignition.  Overlap Begin wont fire if actors are already overlapping.
  (CallFunction|CheckInitialOverlappingActorsNative))

(event Custom|CustomBeginOverlap
  (params (InputPin :type Actor%20Object%20Reference))
  (bind (_asfort_player_pawn _bsuccess) (Utilities|Casting|CastToFortPlayerPawn InputPin))
  ;[0.769 0.09549 1] Curie
  ;;[1 1 1] Begin a Fire Interaction if something touches the actual fire.  End it when they leave.  End all fire interactions when the fire goes out.
  (if (and (== (Variables|StateBools|GetLitRepBoolNative) 1) (Variables|Default|GetCurieActivated))
    (Networking|SwitchHasAuthority
      (:Authority
        (if (not (and (Vehicle|IsinVehicle _asfort_player_pawn) _bsuccess))
          (if (not (Utilities|Map|Contains (Variables|Default|GetInteractHandleMapNative) InputPin))
            (bind _returnvalue (Curie|ExecuteBeginFireInteract (Variables|Interact|GetInteractingPawnNative) self InputPin 1e+02 (Collision|MakeHitResult false false 0.0 0.0 (Transformation|GetActorLocation InputPin) (Transformation|GetActorLocation InputPin) "0,0,1" "0,0,1" 0 InputPin)))
            (Utilities|Map|Add (Variables|Default|GetInteractHandleMapNative) InputPin _returnvalue)))))))

(event OnComponentBeginOverlap(Sphere)
  (params (OverlappedComponent :type Primitive%20Component%20Object%20Reference) (OtherActor :type Actor%20Object%20Reference) (OtherComp :type Primitive%20Component%20Object%20Reference) (OtherBodyIndex :type Integer) (bFromSweep :type Boolean) (SweepResult :type Hit%20Result%20Structure%20%28by%20ref%29))
  ;[0.769 0.09549 1] Curie
  ;; OnComponentBeginOverlap:
  ;; Made this custom event so child blueprints can override this logic
  (CallFunction|CustomBeginOverlap OtherActor))

(event Custom|SpecialS7LightsFireAfterOut
  ;[0.7933 0.4869 1] Season7 Stuff
  ;;[1 1 1] Challenge Progress
  (CallFunction|CheckifUserIsSpecialS7 (Variables|Interact|GetInteractingPawnNative))
  ;;[1 1 1] Add and remove GCs - NO HEALING SHOULD BE DONE IN THIS CASE AS ITS ONLY ON FOR A VERY SHORT TIME
  (Ability|RemoveGameplayCueReplicated (Variables|GameplayCue|GetDousedCueNative))
  (Ability|AddGameplayCueReplicated (Variables|GameplayCue|GetFireCueNative))
  (Development|SetHiddeninGame (Variables|Default|GetWood) false false)
  (Utilities|FlowControl|Delay (Variables|QuestStuff|GetSpecialS7FireTime))
  (Variables|StateBools|SetLitRepBoolNative 0))

(event OnComponentEndOverlap(Sphere)
  (params (OverlappedComponent :type Primitive%20Component%20Object%20Reference) (OtherActor :type Actor%20Object%20Reference) (OtherComp :type Primitive%20Component%20Object%20Reference) (OtherBodyIndex :type Integer))
  (bind _interacthandlemapnative (Variables|Default|GetInteractHandleMapNative))
  (bind (_value _returnvalue) (Utilities|Map|Find _interacthandlemapnative OtherActor))
  ;[0.769 0.09549 1] Curie
  ;;[1 1 1] Begin a Fire Interaction if something touches the actual fire.  End it when they leave.  End all fire interactions when the fire goes out.
  (if (Variables|Default|GetCurieActivated)
    (Networking|SwitchHasAuthority
      (:Authority
        (if _returnvalue
          (Curie|ExecuteEndFireInteract OtherActor _value)
          (Utilities|Map|Remove (Variables|Default|GetInteractHandleMapNative) OtherActor))))))

(event Custom|Server_QuestObjectiveUpdated
  (params (Controller :type Fort%20Player%20Controller%20Object%20Reference) (QuestDef :type Fort%20Quest%20Item%20Definition%20Object%20Reference) (BackendName :type Name) (CompletionCount :type Integer) (ObjectiveCompleted :type Boolean) (QuestCompleted :type Boolean))
  ;[0.7933 0.4869 1] Season7 Stuff
  (if (and (Utilities|String|EqualExactly(String) (Item|GetPersistentName QuestDef) "Quest:quest_br_season7_b") (Utilities|Name|Equal(Name) BackendName "season7_interact_athena_b"))
    (if (or ObjectiveCompleted QuestCompleted)
      (Variables|Default|SetBlockInteraction true)
      (Utilities|FlowControl|Delay 1.0)
      (CallFunction|HandleCharacterVariantTransition (Game|Player|GetPlayerPawn Controller))
      (Utilities|FlowControl|Delay 0.5)
      (Class|FortPlayerController|ChangeActiveVariantforCosmeticItem Controller "AthenaCharacter:cid_332_athena_commando_m_prisoner" "(TagName=\"Cosmetics.Variant.Channel.Progressive\")" "(TagName=\"Cosmetics.Variant.Property.Stage2\")" 0)
      (EventDispatchers|UnbindEventfromOnQuestObjectiveStateChanged Controller (EventDispatchers|CreateEvent :!event Custom|Server_QuestObjectiveUpdated))
      (Variables|Default|SetBlockInteraction false))))

(event Custom|EndFireInteract
  ;[0.769 0.09549 1] Curie
  ;;[1 1 1] Begin a Fire Interaction if something touches the actual fire.  End it when they leave.  End all fire interactions when the fire goes out.
  (CallFunction|EndFireInteractNative))

(event EventDestroyed
  ;[0.769 0.09549 1] Curie
  ;;[1 1 1] Begin a Fire Interaction if something touches the actual fire.  End it when they leave.  End all fire interactions when the fire goes out.
  (CallFunction|EndFireInteract))

(event Custom|HandleCharacterVariantTransition
  (params (TransitioningPawn :type Fort%20Pawn%20Object%20Reference))
  (bind _s7successemitter (Variables|Default|GetS7SuccessEmitter))
  (bind _s7successsound (Variables|Default|GetS7SuccessSound))
  (bind _mesh (Variables|/Script/Engine.Character|GetMesh TransitioningPawn))
  ;[0.7933 0.4869 1] Season7 Stuff
  (Utilities|IsValid TransitioningPawn
    (:Is%20Valid
      (Utilities|IsValid _mesh
        (:Is%20Valid
          (Utilities|IsValid _s7successemitter
            (:Is%20Valid
              (Effects|SpawnEmitterAttached :EmitterTemplate _s7successemitter :AttachToComponent _mesh :AttachPointName "spine_05" :Scale "1.000000,1.000000,1.000000" :LocationType "SnapToTargetIncludingScale" :bAutoDestroy true :PoolingMethod "None" :bAutoActivate true)
              (Utilities|IsValid _s7successsound
                (:Is%20Valid
                  (Audio|PlaySoundatLocation _s7successsound (Transformation|GetActorLocation TransitioningPawn)))))))))))

(event EventOnDeathPlayEffects
  (params (Damage :type Float%20%28single-precision%29) (DamageTags :type Gameplay%20Tag%20Container%20Structure%20%28by%20ref%29) (Momentum :type Vector :default "0, 0, 0") (HitInfo_bBlockingHit :type Boolean) (HitInfo_bInitialOverlap :type Boolean) (HitInfo_Time :type Float%20%28single-precision%29) (HitInfo_Distance :type Float%20%28single-precision%29) (HitInfo_Location :type Vector :default "0, 0, 0") (HitInfo_ImpactPoint :type Vector :default "0, 0, 0") (HitInfo_Normal :type Vector :default "0, 0, 0") (HitInfo_ImpactNormal :type Vector :default "0, 0, 0") (HitInfo_PhysMat :type Physical%20Material%20Object%20Reference) (HitInfo_HitActor :type Actor%20Object%20Reference) (HitInfo_HitComponent :type Primitive%20Component%20Object%20Reference) (HitInfo_HitBoneName :type Name :default "None") (HitInfo_BoneName :type Name :default "None") (HitInfo_HitItem :type Integer) (HitInfo_ElementIndex :type Integer) (HitInfo_FaceIndex :type Integer) (HitInfo_TraceStart :type Vector :default "0, 0, 0") (HitInfo_TraceEnd :type Vector :default "0, 0, 0") (InstigatedBy :type Fort%20Pawn%20Object%20Reference) (DamageCauser :type Actor%20Object%20Reference) (EffectContext :type Gameplay%20Effect%20Context%20Handle%20Structure))
  (Niagara|SpawnSystematLocation :SystemTemplate "/Game/Effects/Fort_Effects/World/Destruction/Base_Effects/NS_General_Deco_Destruction_Large_Converted.NS_General_Deco_Destruction_Large_Converted" :Location HitInfo_Location :Scale "1.000000,1.000000,1.000000" :bAutoDestroy true :bAutoActivate true :PoolingMethod "None" :bPreCullCheck false)
  (Rendering|SetVisibility (Variables|Default|GetFirePitEmpty) false false))

(fn BlueprintCanInteract
  (params (InteractingPawn :type Fort%20Pawn%20Object%20Reference) (InteractionBeingAttempted :type EInteractionBeingAttempted%20Enum :default "FirstInteraction") (InteractionType :type TInteractionType%20Enum :default "IT_NoInteraction"))
  ; FORT-1100305 - Adding HF to disable campfires in the storm
  (if (ScalableFloat|GetValueasBool (Variables|DataRows|GetRow_CampfireDisabledInStorm))
    (bind _returnvalue (SafeZone|IsLocationinSafeZone (Transformation|GetActorLocation)))
    (if (not _returnvalue)
      (return false))
    (else
      _))
  (return (Interaction|BlueprintCanInteractNative InteractingPawn InteractionBeingAttempted InteractionType)))

(fn BlueprintGetFailedInteractionString
  (params (InteractingPawn :type Fort%20Pawn%20Object%20Reference) (InteractionBeingAttempted :type EInteractionBeingAttempted%20Enum :default "FirstInteraction"))
  ; FORT-1100305 - Adding HF to disable campfires in the storm
  (if (ScalableFloat|GetValueasBool (Variables|DataRows|GetRow_CampfireDisabledInStorm))
    (bind _returnvalue (SafeZone|IsLocationinSafeZone (Transformation|GetActorLocation)))
    (if (not _returnvalue)
      (return ""))
    (else
      _))
  (return (Interaction|BlueprintGetFailedInteractionStringNative InteractingPawn InteractionBeingAttempted)))

(fn BlueprintGetInteractionString
  (params (InteractingPawn :type Fort%20Pawn%20Object%20Reference) (InteractionBeingAttempted :type EInteractionBeingAttempted%20Enum :default "FirstInteraction"))
  (return (Interaction|BlueprintGetInteractionStringNative InteractingPawn InteractionBeingAttempted)))

(fn CheckForRetriggerPlaylist
  (bind (_case_0_true _case_1_true) (GameplayTags|TagContainer|CompareTagContainertoOtherTagContainers (Variables|Interact|GetTC_PlaylistNoRetriggerTags) (Variables|Interact|GetTC_PlaylistAllowRetriggerTags) (Class|FortGameStateAthena|GetAthenaPlaylistContextTags (Fortnite|Athena|GameState|GetGameStateAthena))))
  (Variables|Interact|SetbPlaylistBlocksRetriggerNative (and _case_0_true (not _case_1_true))))

(fn CheckIfUserIsSpecialS7
  (params (PawnWhoLitFire :type Fort%20Pawn%20Object%20Reference))
  (bind _returnvalue (GameplayTags|GetOwnedGameplayTags PawnWhoLitFire :!class /Script/GameplayTags.GameplayTagAssetInterface))
  (return (GameplayTags|HasAllTags _returnvalue (Variables|QuestStuff|GetCustomSpecialS7Tags) true)))

(fn DisableIfBelowFlood
  (if (Fortnite|WaterHelpers|IsActorSubmergedinOcean self)
    (Actor|SetLifeSpan 0.2)))

(fn GetCampFireSphereComponent
  (return (Variables|Default|GetSphere)))

(fn GetCampFireWoodComponent
  (return (Variables|Default|GetWood)))

(fn GetCostReActivate
  (return (Math|Float|Ceil (ScalableFloat|GetValueatLevel (Variables|DataRows|GetRow_Light_CostNative)))))

(fn GetLocationToSpawnFireVFX
  (return (Transformation|GetWorldLocation (Variables|Default|GetFireLoc))))

(fn GetWoodComponent
  (return (Variables|Default|GetWood)))

(fn IsStoked
  (return (Variables|Default|GetbStokedNative)))

(fn OnRep_Destroyed
  (if (Variables|StateBools|GetDestroyed)
    (Development|SetHiddeninGame (Variables|Default|GetWood) true false)))

(fn SetSheetValues
  ;[1 1 1] Set Balance Values From Sheets
  (Variables|DousedLife|SetDousedSmokeLifeNative (ScalableFloat|GetValueatLevel (Variables|DousedLife|GetRow_DousedSmokeLife)))
  (Variables|StateBools|SetbHotfixEnabledNative (CallFunction|RowtoBoolNative (Variables|StateBools|GetRow_HotfixEnabled))))

(fn ShouldPlayNativeCurieFX
  (params (FXTag :type Gameplay%20Tag%20Structure))
  (return (Utilities|Struct|MakeFortNativeCurieFXResponse :GameplayCueResponse "IgnoreCue" :bShouldPlayGeneralVFX false :bShouldPlayAmbientAudio false :bShouldPlayGlow false :bShouldPlayFXAsAOE false)))

(fn ConstructionScript)
