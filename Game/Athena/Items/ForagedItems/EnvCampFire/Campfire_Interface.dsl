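; Blueprint interface for the foraged environmental campfire in Athena, declaring queries other code can make about a campfire.
; It exposes three functions: the location to spawn fire VFX, the campfire's wood component, and whether the fire is stoked.
; The bodies here only return empty defaults, with no campfire logic of their own.
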(fn GetLocationToSpawnFireVFX
  (return "0, 0, 0"))

(fn GetWoodComponent
  (return 0))

(fn IsStoked
  (return false))
