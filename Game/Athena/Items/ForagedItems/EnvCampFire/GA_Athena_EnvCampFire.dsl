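; Gameplay ability for the Athena environmental campfire that periodically heals nearby players and vehicles and applies dance quest tag effects, including a night-time variant.
; Event activation reads scalable-float balance values, and a magnitude of 1 stokes the fire for a temporarily faster heal rate.
; A looping timer runs target selection from a height-offset transform, with optional range override through a dynamic effect container.
; It ends on douse, max heals unless infinite, or optionally in storm.
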(event Custom|SetIsStoked
  (params (IsStoked :type Boolean))
  (bind _returnvalue (Ability|GetAvatarActorfromActorInfo))
  ;[1 1 1] Set is Stoked
  (Utilities|IsValid _returnvalue
    (:Is%20Valid
      (bind _asb_bga_athena_env_camp_fire (Utilities|Casting|CastToB_BGA_Athena_EnvCampFire _returnvalue))
      (Variables|/Script/FortniteGame.BuildingGameplayActorCampFire|SetbStokedNative IsStoked _asb_bga_athena_env_camp_fire))))

(event Custom|ApplyDanceTagsOnce
  ;[1 1 1] Season 17 W3 Q1 - Apply Gameplay Ability Tag GE
  (bind (_asynctaskproxy _targetdata _applicationtag) (Ability|Tasks|DoTargetSelection "None" "(TagName=\"Targeting.ApplicationTag.1\")" true)
    (:then
      ;[0.5 0 1] Season 22 Fornitemares - Apply Gameplay Ability Tag GE -- Safe to remove after S22.30
      (if (or (Utilities|Enum|Equal(Enum) (TimeOfDay|GetCurrentDayPhase :!class /Script/FortniteGame.FortKismetLibrary) "Evening") (Utilities|Enum|Equal(Enum) (TimeOfDay|GetCurrentDayPhase :!class /Script/FortniteGame.FortKismetLibrary) "Night"))
        (bind (_asynctaskproxy_2 _targetdata_2 _applicationtag_2) (Ability|Tasks|DoTargetSelection "None" "(TagName=\"Targeting.ApplicationTag.3\")" true)
          (:Targeted
            (Ability|ApplyGameplayEffectContainer _targetdata_2 _applicationtag_2 0))) ; Dance at a Campfire at Night
      ))
    (:Targeted
      (Ability|ApplyGameplayEffectContainer _targetdata _applicationtag 0))) ; Dance Fire
)

(event Custom|FinishStoke
  ;[1 1 1] StokeComplete
  (Variables|Healing|SetTimeBetweenHeals (ScalableFloat|GetValueatLevel (Variables|Rows|GetRow_TimeBetweenHeals)))
  (Utilities|FlowControl|Sequence
    (:then_0
      ;[1 1 1] Reset Heal Counter, targeting location, and Start Healing
      (CallFunction|SetCustomTargetingTrans)
      (Utilities|Time|ClearandInvalidateTimerbyHandle (Variables|Healing|GetTimer_ApplyHeal))
      (bind _returnvalue (Utilities|Time|SetTimerbyEvent (delegate Custom|ApplyHeal) (Variables|Healing|GetTimeBetweenHeals) true))
      (Variables|Healing|SetTimer_ApplyHeal _returnvalue)
      ;[1 1 1] Lets players get quest credit immediately
      ; instead of having to emote after1 second
      (CallFunction|ApplyDanceTagsOnce))
    (:then_1
      (Variables|Default|SetIsStoked false)
      (CallFunction|EndEvent)
      (Utilities|FlowControl|Delay 2.0)
      (CallFunction|SetIsStoked false))))

(event Custom|EndEvent
  (bind _returnvalue (Ability|GetAvatarActorfromActorInfo))
  ;[1 1 1] Check if we have done the max amount of heal applications, if so clear handle and end ability
  (if (>= (stale Custom|ApplyHeal _result) (Variables|Healing|GetMaxHeals))
    ;; FORT-1100305 - Adding HF to allow campfire to remain active indefinitely
    (if (not (ScalableFloat|GetValueasBool (Variables|Rows|GetRow_IsCampfireInfinite)))
      (Utilities|Time|ClearandInvalidateTimerbyHandle (Variables|Healing|GetTimer_ApplyHeal))
      (Utilities|IsValid _returnvalue
        (:Is%20Valid
          (bind _asb_bga_athena_env_camp_fire (Utilities|Casting|CastToB_BGA_Athena_EnvCampFire _returnvalue))
          (Variables|/Script/FortniteGame.BuildingGameplayActorCampFire|SetLitRepBoolNative 0 _asb_bga_athena_env_camp_fire)
          (Variables|/Script/FortniteGame.BuildingGameplayActorCampFire|SetbStokedNative false _asb_bga_athena_env_camp_fire)
          (Ability|EndAbility))))))

(event Custom|ForceEndAbility
  (bind _returnvalue (Ability|GetAvatarActorfromActorInfo))
  ;[1 1 1] Check if we have done the max amount of heal applications, if so clear handle and end ability
  (Utilities|Time|ClearandInvalidateTimerbyHandle (Variables|Healing|GetTimer_ApplyHeal))
  (Utilities|IsValid _returnvalue
    (:Is%20Valid
      (bind _asb_bga_athena_env_camp_fire (Utilities|Casting|CastToB_BGA_Athena_EnvCampFire _returnvalue))
      (Variables|/Script/FortniteGame.BuildingGameplayActorCampFire|SetLitRepBoolNative 0 _asb_bga_athena_env_camp_fire)
      (Variables|/Script/FortniteGame.BuildingGameplayActorCampFire|SetbStokedNative false _asb_bga_athena_env_camp_fire)
      (Ability|EndAbility))))

(event Custom|ApplyHeal
  (bind _returnvalue (Ability|GetAvatarActorfromActorInfo))
  ;[1 1 1] On Tick Apply Heal
  (bind (_asynctaskproxy _targetdata _applicationtag) (Ability|Tasks|DoTargetSelection "None" "(TagName=\"Targeting.ApplicationTag.0\")" true)
    (:then
      ;[1 1 1] Heal Vehicles for a different value
      (bind (_asynctaskproxy_2 _targetdata_2 _applicationtag_2) (Ability|Tasks|DoTargetSelection "None" "(TagName=\"Targeting.ApplicationTag.2\")" true)
        (:Targeted
          (Ability|ApplyGameplayEffectContainer _targetdata_2 _applicationtag_2 0)
          ;[1 1 1] Season 17 W3 Q1 - Apply Gameplay Ability Tag GE
          (bind (_asynctaskproxy_3 _targetdata_3 _applicationtag_3) (Ability|Tasks|DoTargetSelection "None" "(TagName=\"Targeting.ApplicationTag.1\")" true)
            (:then
              ;[0.5 0 1] Season 22 Fornitemares - Apply Gameplay Ability Tag GE -- Safe to remove after S22.30
              (if (or (Utilities|Enum|Equal(Enum) (TimeOfDay|GetCurrentDayPhase :!class /Script/FortniteGame.FortKismetLibrary) "Evening") (Utilities|Enum|Equal(Enum) (TimeOfDay|GetCurrentDayPhase :!class /Script/FortniteGame.FortKismetLibrary) "Night"))
                (bind (_asynctaskproxy_4 _targetdata_4 _applicationtag_4) (Ability|Tasks|DoTargetSelection "None" "(TagName=\"Targeting.ApplicationTag.3\")" true)
                  (:Targeted
                    (Ability|ApplyGameplayEffectContainer _targetdata_4 _applicationtag_4 0))) ; Dance at a Campfire at Night
              ))
            (:Targeted
              (Ability|ApplyGameplayEffectContainer _targetdata_3 _applicationtag_3 0))) ; Dance Fire
        )) ; Vehicle Target Filter
    )
    (:Targeted
      ;[1 0.02985 0.07961] TEMP: Log the campfire actors that are being healed.   DELETE THIS
      (if (== (Utilities|Array|Length (Ability|TargetData|GetAllActorsfromTargetData _targetdata)) 0)
        (Development|PrintString (Utilities|String|Append "Campfire Target Data has " (Utilities|String|ToString(Integer) (Utilities|Array|Length (Ability|TargetData|GetAllActorsfromTargetData _targetdata))) " actors..."))
        (else
          (for (_array_element _array_index) (Ability|TargetData|GetAllActorsfromTargetData _targetdata)
            (Development|PrintString (Utilities|String|Append "Campfire healing actor[" (Utilities|String|ToString(Integer) _array_index) "] = " (Utilities|GetDisplayName _array_element))))
          (Ability|ApplyGameplayEffectContainer _targetdata _applicationtag 0)
          ; FORT-1100305 - Adding HF to disable campfire healing if it's in the storm
          (if (ScalableFloat|GetValueasBool (Variables|Rows|GetRow_CampfireDisabledInStorm))
            (bind _isinsafezone (CallFunction|StormCheck))
            (if (not _isinsafezone)
              (CallFunction|ForceEndAbility)))))
      ;[1 1 1] Check if we have done the max amount of heal applications, if so clear handle and end ability
      (bind _result (Math|Integer|IncrementInt (Variables|Healing|GetCurrentHeals))
        (:%20%20
          (if (Variables|Default|GetIsStoked)
            _
            (elif (>= _result (Variables|Healing|GetMaxHeals))
              ;; FORT-1100305 - Adding HF to allow campfire to remain active indefinitely
              (if (not (ScalableFloat|GetValueasBool (Variables|Rows|GetRow_IsCampfireInfinite)))
                (Utilities|Time|ClearandInvalidateTimerbyHandle (Variables|Healing|GetTimer_ApplyHeal))
                (Utilities|IsValid _returnvalue
                  (:Is%20Valid
                    (bind _asb_bga_athena_env_camp_fire (Utilities|Casting|CastToB_BGA_Athena_EnvCampFire _returnvalue))
                    (Variables|/Script/FortniteGame.BuildingGameplayActorCampFire|SetLitRepBoolNative 0 _asb_bga_athena_env_camp_fire)
                    (Variables|/Script/FortniteGame.BuildingGameplayActorCampFire|SetbStokedNative false _asb_bga_athena_env_camp_fire)
                    (Ability|EndAbility)))))))))) ; Heal Players
)

(event Ability|EventActivateAbilityFromEvent
  (params (EventData :type Gameplay%20Event%20Data%20Structure%20%28by%20ref%29))
  (bind (_eventtag _instigator _target _optionalobject _optionalobject2 _contexthandle _instigatortags _targettags _eventmagnitude) (Utilities|Struct|BreakGameplayEventData EventData))
  (bind _returnvalue (== _eventmagnitude 1.0))
  (bind _returnvalue_2 (Ability|GetAvatarActorfromActorInfo))
  (bind _asb_bga_athena_env_camp_fire (Utilities|Casting|CastToB_BGA_Athena_EnvCampFire _returnvalue_2))
  ;[1 1 1] Reset Heal Counter, targeting location, and Start Healing
  (Ability|Tasks|WaitGameplayEvent "(TagName=\"Event.Athena.EnvCampFire.Douse\")"
    (:then
      (CallFunction|SetBalanceValues _returnvalue)
      (Variables|Default|SetCampfire (Utilities|Casting|CastToB_BGA_Athena_EnvCampFire (Ability|GetAvatarActorfromActorInfo)))
      (if _returnvalue
        (Utilities|Time|SetTimerbyEvent (delegate Custom|FinishStoke) (ScalableFloat|GetValueatLevel (Variables|Rows|GetRow_StokeDuration)))
        (Ability|ExecuteGameplayCueReplicated :self _asb_bga_athena_env_camp_fire :GameplayCueTag (Variables|Default|GetGC_StokeBurst))
        (if (not (Variables|/Script/FortniteGame.BuildingGameplayActorCampFire|GetbStokedNative (Variables|Default|GetCampfire)))
          (CallFunction|SetIsStoked true)
          (Ability|AddGameplayCueReplicated :self _asb_bga_athena_env_camp_fire :GameplayCueTag (Variables|Default|GetGC_StokedLoop))
          (Variables|Default|SetIsStoked true))
        (else
          (Variables|Healing|SetCurrentHeals 0)))
      (CallFunction|SetCustomTargetingTrans)
      (Utilities|Time|ClearandInvalidateTimerbyHandle (Variables|Healing|GetTimer_ApplyHeal))
      (bind _returnvalue_3 (Utilities|Time|SetTimerbyEvent (delegate Custom|ApplyHeal) (Variables|Healing|GetTimeBetweenHeals) true))
      (Variables|Healing|SetTimer_ApplyHeal _returnvalue_3)
      ;[1 1 1] Lets players get quest credit immediately
      ; instead of having to emote after1 second
      (CallFunction|ApplyDanceTagsOnce))
    (:EventReceived
      (CallFunction|ForceEndAbility))))

(event Ability|EventOnEndAbility
  (params (bWasCancelled :type Boolean)))

(fn GetCustomAbilitySourceTransform
  (return (Variables|Targeting|GetCustomTargeting)))

(fn SetBalanceValues
  (params (Stoke :type Boolean))
  ;[1 1 1] Read from sheets and set balance values
  (Variables|Healing|SetMaxHeals (Math|Integer|SelectInt (+ (Variables|Healing|GetMaxHeals) (Math|Float|Truncate (- (/ (ScalableFloat|GetValueatLevel (Variables|Rows|GetRow_StokeDuration)) (ScalableFloat|GetValueatLevel (Variables|Rows|GetRow_Stoke_TimeBetweenHeals))) (/ (ScalableFloat|GetValueatLevel (Variables|Rows|GetRow_StokeDuration)) (ScalableFloat|GetValueatLevel (Variables|Rows|GetRow_TimeBetweenHeals)))))) (Math|Float|Truncate (ScalableFloat|GetValueatLevel (Variables|Rows|GetRow_MaxHeals))) Stoke))
  (Variables|Healing|SetTimeBetweenHeals (Math|Float|SelectFloat (ScalableFloat|GetValueatLevel (Variables|Rows|GetRow_Stoke_TimeBetweenHeals)) (ScalableFloat|GetValueatLevel (Variables|Rows|GetRow_TimeBetweenHeals)) Stoke))
  (if (ScalableFloat|GetValueasBool (Variables|Rows|GetRow_ShouldOverrideCampfireRange))
    (bind _returnvalue (Ability|GetGameplayEffectContainer "(TagName=\"Targeting.ApplicationTag.0\")"))
    (bind _returnvalue_2 (Ability|ReplaceHalfExtentsinGameplayEffectContainer _returnvalue (Math|Vector|MakeVector (ScalableFloat|GetValueatLevel (Variables|Rows|GetRow_OverrideCampfireRange)) (ScalableFloat|GetValueatLevel (Variables|Rows|GetRow_OverrideCampfireRange)) 192.0)))
    (Ability|AddDynamicGameplayEffectContainer (GameplayTags|MakeLiteralGameplayTag "(TagName=\"Targeting.ApplicationTag.0\")") _returnvalue_2)))

(fn SetCustomTargetingTrans
  ;[1 1 1] Apply apply offset to custom targeting location
  (Variables|Targeting|SetCustomTargeting (Math|Transform|MakeTransform (Math|Vector|MakeVector (.x (Transformation|GetActorLocation (Ability|GetAvatarActorfromActorInfo))) (.y (Transformation|GetActorLocation (Ability|GetAvatarActorfromActorInfo))) (+ (.z (Transformation|GetActorLocation (Ability|GetAvatarActorfromActorInfo))) (Variables|Targeting|GetTargetingHeightOffset))) (Transformation|GetActorRotation (Ability|GetAvatarActorfromActorInfo)))))

(fn StormCheck
  (bind _returnvalue (SafeZone|IsLocationinSafeZone (Transformation|GetActorLocation (Ability|GetAvatarActorfromActorInfo))))
  (return _returnvalue))
